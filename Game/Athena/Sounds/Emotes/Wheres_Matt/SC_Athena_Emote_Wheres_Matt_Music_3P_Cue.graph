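# Third-person music cue for the Wheres Matt emote in Athena, routed to the 3P emote music sound class.
# A looping wave player feeds an Enveloper whose volume curve steps from near zero to full a few milliseconds after start.
# It uses shared emote attenuation and emote-music concurrency, keeps playing when silent, and has a submix send to music audio analysis.
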SoundCue /Game/Athena/Sounds/Emotes/Wheres_Matt/SC_Athena_Emote_Wheres_Matt_Music_3P_Cue
root=Enveloper_0
nodes [
 {
  id=Enveloper_0
  children [WavePlayer_0]
  properties {pitchCurve {editorCurveData {keys [{time=0}]}}, volumeCurve {editorCurveData {keys [<at 0.0058 -0.0007>,<at 0.0059 1>]}}}
 }
 {id=WavePlayer_0, properties {bLooping=true, soundWaveAssetPtr=&/Emote_Where_Is_Matt_Mix_01}}
]
properties {
 attenuationSettings=&Fort_Emotes_Athena_Att
 maxDistance=1600
 soundClassObject=/Game/Sounds/FortSoundClassesAndModes/Fort_Music_Emote_3P
 virtualizationMode=PlayWhenSilent
 volumeMultiplier=1
 concurrencyOverrides {resolutionRule=PreventNew}
 concurrencySet [/Game/Sounds/Concurrency_Rules/Athena_EmoteMusic_Conc]
 soundSubmixSends [{minSendLevel=0, soundSubmix=&Fortnite_AudioAnalysis_Music}]
}
